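# Looping indoor wind ambience cue for the CrewCut military base, using indoor ambient sound class, shared ambient attenuation and a loop-limiting concurrency.
# A quality-level root sends its first two slots to a quad wind loop and its last two to a stereo one.
# Each branch lowers pitch through a random modulator, then layers endlessly looping envelopers with differing periods for slow volume and pitch drift, disguising repetition.
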SoundCue /Hermes/Audio/CrewCut/SC_Ambience_CrewCut_MilitaryBase_Wind_Only_Loop_Int
root=QualityLevel_2
nodes [
 {
  id=Enveloper_4
  children [WavePlayer_0]
  properties {
   bLoop=true
   bLoopIndefinitely=true
   loopEnd=30
   pitchCurve {editorCurveData {keys [
      {arriveTangent=-0.0013, interpMode=RCIM_Cubic, leaveTangent=-0.0013, time=0}
      <at 15 0.98 RCIM_Cubic>
      {arriveTangent=0.0013, interpMode=RCIM_Cubic, leaveTangent=0.0013, time=30}
     ]}}
   volumeCurve {editorCurveData {keys [<at 0 1>]}}
  }
 }
 {
  id=Enveloper_5
  children [Enveloper_4]
  properties {
   bLoop=true
   bLoopIndefinitely=true
   loopEnd=20
   pitchCurve {editorCurveData {keys [{time=0}]}}
   volumeCurve {editorCurveData {keys [
      {arriveTangent=-0.01, interpMode=RCIM_Cubic, time=0, value=1}
      <at 10 0.9 RCIM_Cubic>
      {interpMode=RCIM_Cubic, leaveTangent=0.01, time=20, value=1}
     ]}}
  }
 }
 {
  id=Enveloper_6
  children [WavePlayer_1]
  properties {
   bLoop=true
   bLoopIndefinitely=true
   loopEnd=30
   pitchCurve {editorCurveData {keys [
      {arriveTangent=-0.0013, interpMode=RCIM_Cubic, leaveTangent=-0.0013, time=0}
      <at 15 0.98 RCIM_Cubic>
      {arriveTangent=0.0013, interpMode=RCIM_Cubic, leaveTangent=0.0013, time=30}
     ]}}
   volumeCurve {editorCurveData {keys [<at 0 1>]}}
  }
 }
 {
  id=Enveloper_7
  children [Enveloper_6]
  properties {
   bLoop=true
   bLoopIndefinitely=true
   loopEnd=20
   pitchCurve {editorCurveData {keys [{time=0}]}}
   volumeCurve {editorCurveData {keys [
      {arriveTangent=-0.01, interpMode=RCIM_Cubic, time=0, value=1}
      <at 10 0.9 RCIM_Cubic>
      {interpMode=RCIM_Cubic, leaveTangent=0.01, time=20, value=1}
     ]}}
  }
 }
 {id=Modulator_0, children [Enveloper_7], properties {pitchMax=0.95, pitchMin=0.85, volumeMax=1, volumeMin=1}}
 {id=Modulator_1, children [Enveloper_5], properties {pitchMax=0.95, pitchMin=0.85, volumeMax=1}}
 {id=QualityLevel_2, children [Modulator_0,Modulator_0,Modulator_1,Modulator_1]}
 {
  id=WavePlayer_0
  properties {bLooping=true, soundWaveAssetPtr=&/Ambience_CrewCut_MilitaryBase_Exterior_Wind_Loop_Int_Stereo}
 }
 {
  id=WavePlayer_1
  properties {bLooping=true, soundWaveAssetPtr=&/Ambience_CrewCut_MilitaryBase_Exterior_Wind_Loop_Int_Quad}
 }
]
properties {
 attenuationSettings=/AmbientAudioBase/Attenuations/Fort_Ambient_Att
 bEnableBusSends=false
 bEnableSubmixSends=false
 maxDistance=2097152
 priority=4
 soundClassObject=&Fort_Ambient_Indoors_Athena
 volumeMultiplier=0.35
 concurrencySet [/AmbientAudioBase/Concurrency/AmbientConc_2Max_Loop_PreventNew_Default_Concurrency]
}
